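# MF_DigitCount is a library-exposed material function in the vending machine FX folder that returns the number of digits in a scalar input.
# It floors the number at ones, tens and hundreds scale, saturates each result to 0 or 1, and sums the three without branching.
# The count tops out at three digits, and any value below one returns zero.
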MaterialFunction /Game/Creative/Items/Gameplay/VendingMachine/FX/MF_DigitCount
graph {
 properties {bExposeToLibrary=true, description=Returns the number of digits}
 nodes [
  n1 = FunctionInput() {inputName=Number, inputType=FunctionInput_Scalar}
  n2 = Saturate(None=Floor(None=@n1))
  n3 = Saturate(None=Floor(None=Divide(A=@n1) {constB=10}))
  n4 = Floor(None=Divide(A=@n1) {constB=100})
  FunctionOutput(None=Add(A=Add(A=@n2, B=@n3), B=Saturate(None=@n4)))
 ]
}
