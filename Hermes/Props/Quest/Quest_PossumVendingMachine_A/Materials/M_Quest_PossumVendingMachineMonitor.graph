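# Material for the possum quest vending machine monitor, an emissive screen showing a weapon icon, radial timer, item selection dots and metal, wood and stone costs.
# Parameters drive rainbow success, pulsing fail and unaffordable text colors, plus slide and noise distortion of the layout.
# Digit-count functions and a number-sheet texture render costs procedurally.
# Scanlines are added before MF_CreativeMoveTool, and low-quality switches drop the glitch effects.
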Material /Hermes/Props/Quest/Quest_PossumVendingMachine_A/Materials/M_Quest_PossumVendingMachineMonitor
graph {
 properties {bUseMaterialAttributes=true, bUsedWithNanite=true, pixelDepthOffsetMode=PDOM_Legacy, usedShadingModels=Default Lit}
 nodes [
  n1 = VectorParameter() {parameterName=TextColor, defaultValue=<rgba 1 1 1 0>}
  n2 = VectorParameter() {parameterName=TextColorFail, defaultValue=<rgba 1 0.0102 0.0301 0>}
  n3 = MaterialFunctionCall(Speed=Constant() {r=8}) {materialFunction=&DebugTimeSine}
  LinearInterpolate(A=Multiply(A=@n1.RGB) {constB=0.15}, B=@n2.RGB, Alpha=Round(None=@n3.Result))
  MaterialFunctionCall() {materialFunction=&DebugTimeSine}
  MaterialFunctionCall() {materialFunction=&DebugTimeSine}
  MaterialFunctionCall() {materialFunction=&DebugTimeSine}
  MaterialFunctionCall() {materialFunction=&DebugTimeSine}
  n4 = Multiply(A=Time(), B=ScalarParameter() {defaultValue=22, parameterName=Block Speed A})
  n5 = TextureSample(UVs=Panner(Coordinate=TextureCoordinate() {uTiling=0.2}, Time=Ceil(None=@n4)) {bFractionalPart=true, speedX=0.7, speedY=-0.3}) {samplerType=SAMPLERTYPE_LinearColor, texture=/Game/Global/Textures/Default/Noise/T_Atlas_DigitalNoise_01}
  n6 = ScalarParameter() {parameterName=Distort}
  n7 = Multiply(A=@n5.RGB, B=@n6)
  n8 = SphereMask(A=TextureCoordinate(), B=Constant() {r=0.5}, Radius=Constant() {r=0.8}, Hardness=Constant())
  n9 = ScalarParameter() {parameterName=Success}
  n10 = Add(A=Multiply(A=Time()) {constB=-0.5}, B=@n9)
  n11 = Multiply(A=TextureSample(UVs=Panner(Time=@n10) {speedY=-3}) {texture=/Game/Effects/Fort_Effects/Textures/Grad/T_Gradient_Rainbow_NoRed}.RGB) {constB=2.5}
  n12 = Add(A=VectorParameter() {parameterName=ScreenColor, defaultValue=<rgba 0.9184 0.2296 0 0>}.RGB, B=Multiply(A=@n11, B=@n9))
  n13 = Multiply(A=LinearInterpolate(Alpha=@n5.B) {constA=-0.1, constB=0.1}, B=@n6)
  n14 = Multiply(A=Time()) {constB=0.2}
  n15 = Floor(None=Multiply(A=Frac(None=@n14)) {constB=1.1})
  n16 = TextureCoordinate() {vTiling=2.2}
  n17 = Multiply(A=Frac(None=ComponentMask(None=@n16) {g=true})) {constB=3}
  n18 = Multiply(A=Multiply(A=@n15, B=Floor(None=@n17)) {constA=0.3}, B=@n15)
  n19 = AppendVector(A=Clamp(None=Multiply(B=@n18) {constA=-0.03}) {minDefault=-1}, B=Constant())
  n20 = Add(
   A=QualitySwitch(Default=Add(A=@n13, B=@n19), Low=Constant())
   B=Multiply(A=ScalarParameter() {parameterName=Slide}, B=Constant2Vector() {r=1})
  )
  n21 = Multiply(
   A=Add(A=TextureCoordinate(), B=@n20)
   B=MaterialFunctionCall(X=Constant() {r=1}, Y=ScalarParameter() {defaultValue=2, parameterName=HeightScale}) {materialFunction=&MakeFloat2}.Result
  )
  n22 = Constant2Vector() {g=-1}
  n23 = ComponentMask(None=Saturate(None=Add(A=@n21, B=@n22))) {g=true}
  n24 = Saturate(None=ScalarParameter() {parameterName=CostMetalNumber})
  n25 = Saturate(None=ScalarParameter() {parameterName=CostWoodNumber})
  n26 = Saturate(None=ScalarParameter() {parameterName=CostStoneNumber})
  n27 = Add(A=@n24, B=Add(A=@n25, B=@n26))
  n28 = Floor(None=Add(A=@n23, B=Multiply(A=@n27) {constB=0.32}) {constB=0.96})
  n29 = Floor(None=Add(A=@n23) {constB=0})
  n30 = Max(A=@n24, B=@n25)
  n31 = Rotator(Coordinate=Add(A=Multiply(A=@n21) {constB=0.7}, B=Constant2Vector() {g=-0.08, r=0.15})) {speed=-1.5}
  n32 = Add(
   A=Multiply(A=Subtract(A=@n28, B=@n29), B=Max(A=@n30, B=@n26))
   B=TextureSample(UVs=Saturate(None=@n31)) {texture=/Game/Creative/Items/Gameplay/VendingMachine/FX/T_StarShape}.R
  ) {constB=0}
  n33 = Multiply(A=@n8, B=LinearInterpolate(A=@n12, B=Multiply(A=@n12) {constB=0.45}, Alpha=Saturate(None=@n32)) {constAlpha=0})
  n34 = Power(Base=OneMinus(None=@n8))
  n35 = Multiply(A=@n3.Result, B=VectorParameter() {parameterName=ScreenColorFail, defaultValue=<rgba 1 0.1921 0.0676 1>}.RGB)
  n36 = LinearInterpolate(A=@n33, B=Add(A=Multiply(A=@n34, B=@n35), B=Multiply(A=@n33) {constB=0.3}), Alpha=ScalarParameter() {parameterName=Fail})
  n37 = Constant3Vector() {constant=<rgba 1 1 1 0>}
  n38 = VectorParameter() {parameterName=ItemSelectionVector}
  n39 = Add(A=Multiply(A=@n38.R) {constB=0.01}) {constB=0.027}
  n40 = Constant2Vector() {g=0.1, r=0.4}
  n41 = Multiply(
   A=TextureCoordinate()
   B=MaterialFunctionCall(X=Constant() {r=1}, Y=ScalarParameter() {defaultValue=2, parameterName=HeightScale}) {materialFunction=&MakeFloat2}.Result
  )
  n42 = ScalarParameter() {parameterName=NumItems}
  n43 = Multiply(B=OneMinus(None=Saturate(None=@n42))) {constA=0.015}
  n44 = Constant2Vector() {g=0.1, r=0.5}
  n45 = SphereMask(A=@n41, B=@n44, Radius=Add(A=Multiply(A=@n38.G) {constB=0.01}) {constB=0.027}) {attenuationRadius=0.027}
  n46 = Saturate(None=Subtract(A=@n42))
  n47 = SphereMask(A=@n41, B=@n44, Radius=Multiply(B=OneMinus(None=@n46)) {constA=0.015})
  n48 = Constant2Vector() {g=0.1, r=0.6}
  n49 = SphereMask(A=@n41, B=@n48, Radius=Add(A=Multiply(A=@n38.B) {constB=0.01}) {constB=0.027}) {attenuationRadius=0.027}
  n50 = Saturate(None=Subtract(A=@n42) {constB=2})
  n51 = SphereMask(A=@n41, B=@n48, Radius=Multiply(B=OneMinus(None=@n50)) {constA=0.015})
  n52 = Add(
   A=Subtract(A=SphereMask(A=@n41, B=@n40, Radius=@n39) {attenuationRadius=0.027}, B=SphereMask(A=@n41, B=@n40, Radius=@n43))
   B=Add(A=Subtract(A=@n45, B=@n47), B=Subtract(A=@n49, B=@n51))
  )
  n53 = Divide(A=Subtract(A=Time(), B=ScalarParameter() {parameterName=TimerStartTime}), B=ScalarParameter() {defaultValue=5, parameterName=TimerDuration})
  n54 = TextureCoordinate()
  n55 = MaterialFunctionCall(X=Constant() {r=1}, Y=ScalarParameter() {defaultValue=2, parameterName=HeightScale}) {materialFunction=&MakeFloat2}
  n56 = Add(
   A=Add(A=Multiply(A=@n54, B=@n55.Result), B=@n20)
   B=MaterialFunctionCall(Y=Constant() {r=-0.28}) {materialFunction=&MakeFloat2}.Result
  )
  n57 = Saturate(None=@n56)
  n58 = Add(A=@n57, B=Constant2Vector() {g=-0.045})
  n59 = MaterialFunctionCall(UVs=@n58, "Rotation Angle (0-1)"=ScalarParameter() {defaultValue=0.25, parameterName=Origin_0-1}) {materialFunction=&CustomRotator}
  n60 = Add(
   A=OneMinus(None=Clamp(None=@n53))
   B=MaterialFunctionCall(Vector or UVs=Subtract(A=@n59.Rotated Values) {constB=0.5}) {materialFunction=&VectorToRadialValue}.Vector Converted to Angle
  )
  n61 = Clamp(None=Multiply(A=ScalarParameter() {defaultValue=0.77, parameterName=OuterRadius_0-1}) {constB=0.5}) {maxDefault=0.5}
  n62 = Clamp(None=Multiply(A=MaterialFunctionCall!(UVs=@n58, CenterPosition, Radius=@n61) {materialFunction=/Engine/Functions/Engine_MaterialFunctions01/Gradient/SphereGradient-2D}.Result 0-1) {constB=2})
  n63 = Clamp(None=Multiply(A=ScalarParameter() {defaultValue=0.55, parameterName=InnerRadius_0-1}) {constB=0.5}) {maxDefault=0.495}
  n64 = Clamp(None=Multiply(A=MaterialFunctionCall(UVs=@n58, Radius=@n63) {materialFunction=&SphereGradient-2D}.Result 0-1) {constB=2})
  n65 = ComponentMask(None=Multiply(A=Floor(None=@n60), B=Subtract(A=@n62, B=@n64))) {r=true}
  n66 = Clamp(None=AppendVector(A=AppendVector(A=@n65, B=@n65), B=@n65))
  n67 = Multiply(A=VectorParameter() {parameterName=ScreenColor, defaultValue=<rgba 0.9184 0.2296 0 0>}.RGB) {constB=0.3}
  n68 = TextureSampleParameter2D(UVs=@n57) {parameterName=WeaponIcon, texture=/Hermes/Props/Quest/Quest_PossumVendingMachine_A/Textures/T_Quest_PossumVendingMachine_Icon}
  n69 = Multiply(A=Multiply(
    A=ScalarParameter() {parameterName=Success}
    B=MaterialFunctionCall(Hue Shift Percentage=@n68.RGB, Texture=@n11) {materialFunction=&HueShift}.Result
   ) {constB=0.25}) {constB=0.2}
  n70 = MaterialFunctionCall(Float2=@n56) {materialFunction=&BreakOutFloat2Components}
  n71 = OneMinus(None=@n70.R)
  n72 = Round(None=Add(A=Min(A=@n70.R, B=@n71)) {constB=0.5})
  n73 = OneMinus(None=@n70.G)
  n74 = Round(None=Add(A=Min(A=@n70.G, B=@n73)) {constB=0.5})
  n75 = LinearInterpolate(
   A=Add(A=Multiply(A=@n66, B=@n67) {constA=1, constB=0.2}, B=LinearInterpolate(A=@n36, B=@n37, Alpha=@n52))
   B=QualitySwitch(Default=Add(A=@n68.RGB, B=@n69), Low=@n68.RGB)
   Alpha=Multiply(A=Min(A=@n72, B=@n74), B=@n68.A)
  )
  n76 = Constant() {r=1}
  n77 = ScalarParameter() {defaultValue=2, parameterName=HeightScale}
  n78 = Constant() {r=3.5}
  n79 = Multiply(
   A=Multiply(
    A=TextureCoordinate()
    B=MaterialFunctionCall(X=@n76, Y=@n77) {materialFunction=&MakeFloat2}.Result
   )
   B=@n78
  )
  n80 = Constant2Vector() {g=-6, r=-0.35}
  n81 = TextureSampleParameter2D(UVs=Add(A=Add(A=@n79, B=@n80), B=Multiply(A=@n20, B=@n78))) {parameterName=ResourceIconMetal, samplerSource=SSM_Clamp_WorldGroupSettings, texture=/Game/UI/Icons/Icon-Resource/Icon-Resource-Metal-M}
  n82 = Multiply(A=@n81.A, B=Saturate(None=ScalarParameter() {parameterName=CostMetalNumber}))
  n83 = TextureCoordinate()
  n84 = MaterialFunctionCall(X=Constant() {r=1}, Y=ScalarParameter() {defaultValue=2, parameterName=HeightScale}) {materialFunction=&MakeFloat2}
  n85 = Constant() {r=3.5}
  n86 = Add(A=Multiply(A=Multiply(A=@n83, B=@n84.Result), B=@n85), B=Constant2Vector() {g=-6, r=-0.35})
  n87 = Multiply(A=Saturate(None=ScalarParameter() {parameterName=CostMetalNumber}), B=Constant2Vector() {g=1})
  n88 = TextureSampleParameter2D(UVs=Add(A=Add(A=@n86, B=@n87), B=Multiply(A=@n20, B=@n85))) {parameterName=ResourceIcon, samplerSource=SSM_Clamp_WorldGroupSettings, texture=/Game/UI/Icons/Icon-Resource/Icon-Resource-Wood-M}
  n89 = Saturate(None=ScalarParameter() {parameterName=CostWoodNumber})
  n90 = ScalarParameter() {parameterName=CostWoodNumber}
  n91 = Add(A=Multiply(A=Saturate(None=@n90), B=Constant2Vector() {g=1}), B=@n87)
  n92 = TextureCoordinate()
  n93 = MaterialFunctionCall(X=Constant() {r=1}, Y=ScalarParameter() {defaultValue=2, parameterName=HeightScale}) {materialFunction=&MakeFloat2}
  n94 = Constant() {r=3.5}
  n95 = Add(A=Multiply(A=Multiply(A=@n92, B=@n93.Result), B=@n94), B=Constant2Vector() {g=-6, r=-0.35})
  n96 = TextureSampleParameter2D(UVs=Add(A=Add(A=@n91, B=@n95), B=Multiply(B=@n94))) {parameterName=ResourceIconStone, samplerSource=SSM_Clamp_WorldGroupSettings, texture=/Game/UI/Icons/Icon-Resource/Icon-Resource-Stone-M}
  n97 = ScalarParameter() {parameterName=CostStoneNumber}
  n98 = LinearInterpolate(
   A=LinearInterpolate(A=LinearInterpolate(A=@n75, B=@n81.RGB, Alpha=@n82), B=@n88.RGB, Alpha=Multiply(A=@n88.A, B=@n89))
   B=@n96.RGB
   Alpha=Multiply(A=@n96.A, B=Saturate(None=@n97))
  )
  n99 = Multiply(A=@n98, B=Add(A=@n7))
  n100 = Multiply(A=@n7) {constB=0.3}
  n101 = ScalarParameter() {parameterName=Unaffordable}
  n102 = Constant2Vector()
  n103 = Saturate(None=ScalarParameter() {parameterName=CostWoodNumber})
  n104 = Saturate(None=ScalarParameter() {parameterName=CostMetalNumber})
  n105 = Add(A=Constant2Vector() {g=0.86, r=0.4}, B=Multiply(A=Add(A=@n103, B=@n104), B=Constant2Vector() {g=-0.14}))
  n106 = ScalarParameter() {parameterName=CostStoneNumber}
  n107 = Multiply(
   A=Constant2Vector() {r=0.06}
   B=Subtract(
    A=Constant() {r=3}
    B=MaterialFunctionCall!(Number=@n106) {materialFunction=/Game/Creative/Items/Gameplay/VendingMachine/FX/MF_DigitCount}.Result
   )
  )
  n108 = Constant2Vector() {g=0.11, r=0.35}
  n109 = MaterialFunctionCall!(
   UVs=Add(A=TextureCoordinate(), B=@n20)
   TextureIn
   UpperLeftCorner=Add(A=Add(A=@n102, B=@n105), B=@n107)
   LowerRightCorner=Subtract(A=Add(A=@n108, B=@n105), B=@n107)
   ExportFloat4
  ) {materialFunction=/Engine/Functions/Engine_MaterialFunctions01/Texturing/TextureCropping}
  n110 = Multiply(A=@n109.Crop UVs, B=Constant2Vector() {g=0.1, r=1}) {constA=2, constB=0.1}
  n111 = Multiply(
   A=OneMinus(None=ComponentMask(None=@n110) {r=true})
   B=MaterialFunctionCall(Number=ScalarParameter() {parameterName=CostStoneNumber}) {materialFunction=&MF_DigitCount}.Result
  ) {constB=10}
  n112 = Ceil(None=Max(A=@n111) {constB=0})
  n113 = Divide(A=ScalarParameter() {parameterName=CostStoneNumber}, B=Power(Base=Constant() {r=10}, Exp=Add(A=@n112)) {desc="10-100,000,000"})
  n114 = Floor(None=Multiply(A=Frac(None=@n113)) {constB=100})
  n115 = AppendVector(A=Frac(None=@n111), B=Add(A=ComponentMask(None=@n110) {g=true}, B=Divide(A=@n114) {constB=10}))
  n116 = Saturate(None=Multiply(
    A=@n109.Crop Mask
    B=TextureSample!(UVs=@n115, Tex, Level, Apply View MipBias) {mipValueMode=TMVM_MipLevel, samplerType=SAMPLERTYPE_Masks, texture=/Game/Athena/Items/Gameplay/VendingMachine/FX/T_NumberSheet}.R
   ) {desc=crop mask})
  n117 = Saturate(None=ScalarParameter() {parameterName=CostStoneNumber})
  n118 = Constant2Vector()
  n119 = Add(A=Constant2Vector() {g=0.86, r=0.4}, B=Multiply(A=@n104, B=Constant2Vector() {g=-0.14}))
  n120 = ScalarParameter() {parameterName=CostWoodNumber}
  n121 = Multiply(
   A=Constant2Vector() {r=0.06}
   B=Subtract(
    A=Constant() {r=3}
    B=MaterialFunctionCall(Number=@n120) {materialFunction=&MF_DigitCount}.Result
   )
  )
  n122 = Constant2Vector() {g=0.11, r=0.35}
  n123 = MaterialFunctionCall(
   UVs=Add(A=TextureCoordinate(), B=@n20)
   UpperLeftCorner=Add(A=Add(A=@n118, B=@n119), B=@n121)
   LowerRightCorner=Subtract(A=Add(A=@n122, B=@n119), B=@n121)
  ) {materialFunction=&TextureCropping}
  n124 = Multiply(A=@n123.Crop UVs, B=Constant2Vector() {g=0.1, r=1}) {constA=2, constB=0.1}
  n125 = Multiply(
   A=OneMinus(None=ComponentMask(None=@n124) {r=true})
   B=MaterialFunctionCall(Number=ScalarParameter() {parameterName=CostWoodNumber}) {materialFunction=&MF_DigitCount}.Result
  ) {constB=10}
  n126 = ComponentMask(None=@n124) {g=true}
  n127 = Constant() {r=10}
  n128 = Add(A=Ceil(None=Max(A=@n125) {constB=0}))
  n129 = Frac(None=Divide(A=ScalarParameter() {parameterName=CostWoodNumber}, B=Power(Base=@n127, Exp=@n128) {desc="10-100,000,000"}))
  n130 = Divide(A=Floor(None=Multiply(A=@n129) {constB=100})) {constB=10}
  n131 = TextureSample(UVs=AppendVector(A=Frac(None=@n125), B=Add(A=@n126, B=@n130))) {mipValueMode=TMVM_MipLevel, samplerType=SAMPLERTYPE_Masks, texture=&T_NumberSheet}
  n132 = Multiply(A=Saturate(None=Multiply(A=@n123.Crop Mask, B=@n131.R) {desc=crop mask}), B=Saturate(None=ScalarParameter() {parameterName=CostWoodNumber}))
  n133 = Constant2Vector()
  n134 = Constant2Vector() {g=0.86, r=0.4}
  n135 = ScalarParameter() {parameterName=CostMetalNumber}
  n136 = Multiply(
   A=Constant2Vector() {r=0.06}
   B=Subtract(
    A=Constant() {r=3}
    B=MaterialFunctionCall(Number=@n135) {materialFunction=&MF_DigitCount}.Result
   )
  )
  n137 = Constant2Vector() {g=0.11, r=0.35}
  n138 = MaterialFunctionCall(
   UVs=Add(A=TextureCoordinate(), B=@n20)
   UpperLeftCorner=Add(A=Add(A=@n133, B=@n134), B=@n136)
   LowerRightCorner=Subtract(A=Add(A=@n137, B=@n134), B=@n136)
  ) {materialFunction=&TextureCropping}
  n139 = Multiply(A=@n138.Crop UVs, B=Constant2Vector() {g=0.1, r=1}) {constA=2, constB=0.1}
  n140 = Multiply(
   A=OneMinus(None=ComponentMask(None=@n139) {r=true})
   B=MaterialFunctionCall(Number=ScalarParameter() {parameterName=CostMetalNumber}) {materialFunction=&MF_DigitCount}.Result
  ) {constB=10}
  n141 = ComponentMask(None=@n139) {g=true}
  n142 = Constant() {r=10}
  n143 = Add(A=Ceil(None=Max(A=@n140) {constB=0}))
  n144 = Frac(None=Divide(A=ScalarParameter() {parameterName=CostMetalNumber}, B=Power(Base=@n142, Exp=@n143) {desc="10-100,000,000"}))
  n145 = Divide(A=Floor(None=Multiply(A=@n144) {constB=100})) {constB=10}
  n146 = TextureSample(UVs=AppendVector(A=Frac(None=@n140), B=Add(A=@n141, B=@n145))) {mipValueMode=TMVM_MipLevel, samplerType=SAMPLERTYPE_Masks, texture=&T_NumberSheet}
  n147 = Multiply(A=Saturate(None=Multiply(A=@n138.Crop Mask, B=@n146.R) {desc=crop mask}), B=Saturate(None=ScalarParameter() {parameterName=CostMetalNumber}))
  n148 = LinearInterpolate(
   A=QualitySwitch(Default=Add(A=@n99, B=@n100), Low=@n98)
   B=LinearInterpolate(A=@n1.RGB, B=@n2.RGB, Alpha=Ceil(None=@n101)) {constAlpha=1}
   Alpha=Add(A=Multiply(A=@n116, B=@n117), B=Add(A=@n132, B=@n147))
  )
  n149 = Constant() {r=1}
  n150 = ScalarParameter() {defaultValue=2, parameterName=HeightScale}
  n151 = Panner(Coordinate=Multiply(
    A=TextureCoordinate() {vTiling=32}
    B=MaterialFunctionCall(X=@n149, Y=@n150) {materialFunction=&MakeFloat2}.Result
   )) {speedX=0.5, speedY=0.7}
  n152 = Multiply(
   A=ScalarParameter() {defaultValue=1, parameterName=ScreenBrightness}
   B=Multiply(A=TextureSample!(UVs=@n151, Tex, Apply View MipBias) {texture=/Game/Missions/Primary/LaunchTheBalloon/Textures/T_TV_Scanlines}.RGB) {constB=0.04}
  )
  MaterialFunctionCall(MA In=SetMaterialAttributes(Emissive Color=Add(A=@n148, B=@n152)) {attributeSetTypes [#1]}) {materialFunction=&MF_CreativeMoveTool}
 ]
}
